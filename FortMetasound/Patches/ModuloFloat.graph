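# MetaSound patch that computes a floating-point modulo of A by B, wrapping the result toward the zero-to-B range.
# It splits off A's fractional part, applies integer modulo to the truncated A and B, then adds the fraction back.
# When that sum changes, chained trigger compares choose between adding B, subtracting B or passing it through, and a trigger route drives Out.
# Any fractional part of B is ignored in the modulo step.
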MetaSoundPatch /FortMetasound/Patches/ModuloFloat
inputs [
 A {type=Float}
 B {type=Float}
]
interfaces [MetaSound:1.0]
outputs [Out {type=Float}]
pages [Default {
  input_defaults {A=0, B=180}
  nodes [
   InitVariable.Float()
   n1 = Convert.Float:Int32(Int32=Convert.Int32:Float(Float=^A).Int32)
   n2 = VariableAccessor.Float(Variable=VariableMutator.Float(Variable=InitVariable.Float().Variable, Value=UE.Subtract:Float(PrimaryOperand=^A, AdditionalOperands=@n1.Float).Out).Variable)
   n3 = UE.Modulo:Int32(AdditionalOperands=Convert.Int32:Float(Float=^B).Int32, PrimaryOperand=Convert.Int32:Float(Float=^A).Int32)
   n4 = VariableAccessor.Float(Variable=VariableMutator.Float(Value=Convert.Float:Int32(Int32=@n3.Out).Float, Variable=InitVariable.Float().Variable).Variable)
   n5 = VariableAccessor.Float(Variable=VariableMutator.Float(Value=UE.Add:Float(AdditionalOperands=@n2.Value, PrimaryOperand=@n4.Value).Out, Variable=InitVariable.Float().Variable).Variable)
   n6 = VariableAccessor.Float(Variable=@n5.Variable)
   n7 = VariableAccessor.Float(Variable=@n6.Variable)
   TriggerOnValueChange_0 = UE.Trigger On Value Change:Float(Value=@n5.Value)
   n8 = TriggerCompare.Clamp:Float(A=UE.Reroute(Value=@n5.Value).Value, Compare=@TriggerOnValueChange_0.Trigger) {also {defaults {Type=2}}}
   n9 = VariableAccessor.Float(Variable=@n7.Variable)
   n10 = VariableAccessor.Float(Variable=@n9.Variable)
   n11 = TriggerCompare.Clamp:Float(B=^B, Compare=@n8.FALSE, A=VariableAccessor.Float(Variable=@n10.Variable).Value) {also {defaults {Type=5}}}
   n12 = Value.Value:Float(Target Value=UE.Add:Float(AdditionalOperands=^B, PrimaryOperand=@n10.Value).Out, Set=@n8.TRUE, Init Value=@n9.Value)
   TriggerRouteTriggerRouteFloat3_0 = "TriggerRoute.Trigger Route (Float, 3):Float"(
    Value 1=UE.Subtract:Float(AdditionalOperands=^B, PrimaryOperand=@n7.Value).Out
    Set 2=@n11.FALSE
    Set 1=@n11.TRUE
    Set 0=@n12.On Set
    Value 0=@n12.Output Value
    Value 2=@n6.Value
   )
  ]
  links [{dst="@Out", src=TriggerRouteTriggerRouteFloat3_0.Value}]
 }]
