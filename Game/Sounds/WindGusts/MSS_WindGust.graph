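# One-shot stereo MetaSound for outdoor wind gusts, playing a random leaf-wind wave, limited to one instance.
# Randomised pitch and volume feed the AB_Wind bus, then each channel pans by listener azimuth plus random offset and spread, faking a world-fixed direction.
# The enclosure modulator attenuates and filters the mix before a Duration-timed ADSR fades it and signals finish.
# The Mobile page uses fewer waves and omits randomised pitch, volume, direction and enclosure processing.
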MetaSoundSource /Game/Sounds/WindGusts/MSS_WindGust
inputs [
 Duration {type=Float}
 Enclosed.LPF {type=Float}
 Enclosed.Volume {type=Float}
 MasterVolume {type=Float}
 Release Time {type=Time}
 Sounds {type=WaveAsset:Array}
 StereoSpread {type=Float}
 UE.Listener.Orientation.Azimuth {type=Float}
 UE.Listener.Orientation.Elevation {type=Float}
 UE.Source.OnPlay {type=Trigger}
]
interfaces [UE.Listener.Orientation:1.0, UE.OutputFormat.Stereo:1.0, UE.Source.OneShot:1.0, UE.Source:1.1]
outputs [
 "UE.OutputFormat.Stereo.Audio:0" {type=Audio}
 "UE.OutputFormat.Stereo.Audio:1" {type=Audio}
 UE.Source.OneShot.OnFinished {type=Trigger}
]
pages [
 Default {
  is_build_graph=true
  input_defaults {
   Duration=7
   Enclosed.LPF=6000
   Enclosed.Volume=0.1
   MasterVolume=1
   Release Time=2
   StereoSpread=90
   UE.Listener.Orientation.Azimuth=360
   UE.Listener.Orientation.Elevation=0
   UE.Source.OnPlay=null
   Sounds [
    /Game/Sounds/Artemis/Biomes/Reality_Forest/Spots/AMB_Biome_Reality_Forest_SS_Wind_Leaves_01_01
    /Game/Sounds/Artemis/Biomes/Reality_Forest/Spots/AMB_Biome_Reality_Forest_SS_Wind_Leaves_01_08
    /Game/Sounds/Artemis/Biomes/Reality_Forest/Spots/AMB_Biome_Reality_Forest_SS_Wind_Leaves_01_07
    /Game/Sounds/Artemis/Biomes/Reality_Forest/Spots/AMB_Biome_Reality_Forest_SS_Wind_Leaves_01_06
    /Game/Sounds/Artemis/Biomes/Reality_Forest/Spots/AMB_Biome_Reality_Forest_SS_Wind_Leaves_01_05
    /Game/Sounds/Artemis/Biomes/Reality_Forest/Spots/AMB_Biome_Reality_Forest_SS_Wind_Leaves_01_04
   ]
  }
  nodes [
   n1 = InitVariable.Float()
   n2 = UE.RandomFloat(Next=^UE.Source.OnPlay) {also {defaults {Max=0.6, Min=0.4}}}
   n3 = UE.Multiply:Float(
    PrimaryOperand=^MasterVolume
    AdditionalOperands=VariableAccessor.Float(Variable=VariableMutator.Float(Variable=@n1.Variable, Value=@n2.Value).Variable).Value
   )
   ArrayRandomGet_0 = Array.Random Get:WaveAsset:Array(In Array=^Sounds, Next=^UE.Source.OnPlay, No Repeats=UE.Divide:Int32(PrimaryOperand=Array.Num:WaveAsset:Array(Array=^Sounds).Num).Out)
   n4 = UE.RandomFloat(Next=@ArrayRandomGet_0.On Next) {also {defaults {Min=-1}}}
   WavePlayer_0 = UE.Wave Player:Stereo(Wave Asset=@ArrayRandomGet_0.Value, Play=@n4.On Next, Pitch Shift=@n4.Value)
   n5 = UE.Multiply:Audio by Float(AdditionalOperands=@n3.Out, PrimaryOperand=@WavePlayer_0.Out Right)
   n6 = UE.Multiply:Audio by Float(AdditionalOperands=@n3.Out, PrimaryOperand=@WavePlayer_0.Out Left)
   AudioBusWriter2_0 = "UE.Audio Bus Writer (2)"(In 1=@n5.Out, In 0=@n6.Out) {also {defaults {Audio Bus=&/AB_Wind}}}
   TriggerDelay_0 = UE.Trigger Delay(
    In=^UE.Source.OnPlay
    Delay Time=UE.Subtract:Time(AdditionalOperands=^Release Time, PrimaryOperand=UE.ConversionFloatToTime(In=^Duration).Out) {also {version [1,1]}}.Out
   )
   ADSREnvelopeADSREnvelope_0 = ADSR Envelope.ADSR Envelope:Audio(
    Trigger Attack=^UE.Source.OnPlay
    Release Time=UE.Add:Time(PrimaryOperand=^Release Time) {also {defaults {AdditionalOperands=2}}}.Out
    Trigger Release=@TriggerDelay_0.Out
   ) {also {defaults {Decay Curve=0.5, Sustain Level=1}}}
   n7 = Modulation.GetModulatorValue() {also {defaults {Modulator=&AudioWorldization_Bus_Enclosure}}}
   n8 = MapRange.MapRange:Float(Out Range B=^Enclosed.LPF, In=@n7.Out) {also {defaults {Out Range A=20000}}}
   n9 = /FortMetasound/Patches/ModuloFloat(A=UE.Add:Float(
     PrimaryOperand=^UE.Listener.Orientation.Azimuth
     AdditionalOperands=UE.RandomFloat(Next=@n2.On Next) {also {defaults {Max=225, Min=45}}}.Value
    ).Out) {also {defaults {B=360}}}
   n10 = UE.Add:Float(AdditionalOperands=^StereoSpread, PrimaryOperand=@n9.Out)
   n11 = /FortMetasound/Patches/CompareFloat(A=@n10.Out) {also {defaults {B=180, Type=3}}}
   n12 = UE.Add:Float(
    PrimaryOperand=@n10.Out
    AdditionalOperands=UE.Multiply:Float(
     PrimaryOperand=Convert.Float:Bool(Bool=@n11.Result).Float
     AdditionalOperands=UE.Multiply:Float() {also {defaults {AdditionalOperands=-2, PrimaryOperand=180}}}.Out
    ).Out
   )
   n13 = UE.Multiply:Float() {also {defaults {AdditionalOperands=-1, PrimaryOperand=180}}}
   n14 = UE.Multiply:Float(
    PrimaryOperand=Convert.Float:Bool(Bool=&CompareFloat(A=@n10.Out, B=@n13.Out) {also {defaults {Type=2}}}.Result).Float
    AdditionalOperands=UE.Multiply:Float() {also {defaults {AdditionalOperands=2, PrimaryOperand=180}}}.Out
   )
   CrossfadeTriggerRouteFloat5_1 = "Crossfade.Trigger Route (Float, 5):Float"(Crossfade Value=MapRange.MapRange:Float(In=UE.Add:Float(PrimaryOperand=@n12.Out, AdditionalOperands=@n14.Out).Out) {also {defaults {In Range A=-180, In Range B=180, Out Range B=4}}}.Out Value) {also {defaults {In 0=0, In 1=-1, In 2=0, In 3=1, In 4=0}}}
   StereoPanner_0 = UE.Stereo Panner(Pan Amount=@CrossfadeTriggerRouteFloat5_1.Out, In=@n5.Out)
   n15 = UE.Subtract:Float(AdditionalOperands=^StereoSpread, PrimaryOperand=@n9.Out)
   n16 = &CompareFloat(A=@n15.Out) {also {defaults {B=180, Type=3}}}
   n17 = UE.Add:Float(
    AdditionalOperands=UE.Multiply:Float(
     PrimaryOperand=Convert.Float:Bool(Bool=@n16.Result).Float
     AdditionalOperands=UE.Multiply:Float() {also {defaults {AdditionalOperands=-2, PrimaryOperand=180}}}.Out
    ).Out
    PrimaryOperand=@n15.Out
   )
   n18 = UE.Multiply:Float() {also {defaults {AdditionalOperands=-1, PrimaryOperand=180}}}
   n19 = UE.Multiply:Float(
    PrimaryOperand=Convert.Float:Bool(Bool=&CompareFloat(B=@n18.Out, A=@n15.Out) {also {defaults {Type=2}}}.Result).Float
    AdditionalOperands=UE.Multiply:Float() {also {defaults {AdditionalOperands=2, PrimaryOperand=180}}}.Out
   )
   CrossfadeTriggerRouteFloat5_0 = "Crossfade.Trigger Route (Float, 5):Float"(Crossfade Value=MapRange.MapRange:Float(In=UE.Add:Float(PrimaryOperand=@n17.Out, AdditionalOperands=@n19.Out).Out) {also {defaults {In Range A=-180, In Range B=180, Out Range B=4}}}.Out Value) {also {defaults {In 0=0, In 1=-1, In 2=0, In 3=1, In 4=0}}}
   StereoPanner_1 = UE.Stereo Panner(Pan Amount=@CrossfadeTriggerRouteFloat5_0.Out, In=@n6.Out)
   AudioMixerAudioMixerStereo2_0 = "AudioMixer.Audio Mixer (Stereo, 2)"(In 1 L=@StereoPanner_0.Out Left, In 1 R=@StereoPanner_0.Out Right, In 0 L=@StereoPanner_1.Out Left, In 0 R=@StereoPanner_1.Out Right)
   n20 = UE.Subtract:Float(AdditionalOperands=@n7.Out) {also {defaults {PrimaryOperand=1}}}
   BiquadFilter_0 = UE.Biquad Filter:Audio(
    Cutoff Frequency=@n8.Out Value
    In=UE.Multiply:Audio by Float(PrimaryOperand=@AudioMixerAudioMixerStereo2_0.Out L, AdditionalOperands=@n20.Out).Out
   ) {also {defaults {Type=0}}}
   Multiply_0 = UE.Multiply:Audio(PrimaryOperand=@ADSREnvelopeADSREnvelope_0.Out Envelope, AdditionalOperands=@BiquadFilter_0.Out)
   BiquadFilter_1 = UE.Biquad Filter:Audio(
    Cutoff Frequency=@n8.Out Value
    In=UE.Multiply:Audio by Float(PrimaryOperand=@AudioMixerAudioMixerStereo2_0.Out R, AdditionalOperands=@n20.Out).Out
   )
   Multiply_1 = UE.Multiply:Audio(PrimaryOperand=@ADSREnvelopeADSREnvelope_0.Out Envelope, AdditionalOperands=@BiquadFilter_1.Out)
  ]
  links [
   {dst="@UE.Source.OneShot.OnFinished", src=ADSREnvelopeADSREnvelope_0.On Done}
   {dst="@UE.OutputFormat.Stereo.Audio:0", src=Multiply_0.Out}
   {dst="@UE.OutputFormat.Stereo.Audio:1", src=Multiply_1.Out}
  ]
 }
 Mobile {
  input_defaults {Sounds [
    &AMB_Biome_Reality_Forest_SS_Wind_Leaves_01_01
    &AMB_Biome_Reality_Forest_SS_Wind_Leaves_01_08
    &AMB_Biome_Reality_Forest_SS_Wind_Leaves_01_07
   ]}
  nodes [
   ArrayRandomGet_0 = Array.Random Get:WaveAsset:Array(In Array=^Sounds, Next=^UE.Source.OnPlay, No Repeats=UE.Divide:Int32(PrimaryOperand=Array.Num:WaveAsset:Array(Array=^Sounds).Num).Out)
   WavePlayer_0 = UE.Wave Player:Stereo(Play=@ArrayRandomGet_0.On Next, Wave Asset=@ArrayRandomGet_0.Value)
   n1 = UE.Multiply:Audio by Float(AdditionalOperands=^MasterVolume, PrimaryOperand=@WavePlayer_0.Out Right)
   n2 = UE.Multiply:Audio by Float(AdditionalOperands=^MasterVolume, PrimaryOperand=@WavePlayer_0.Out Left)
   AudioBusWriter2_0 = "UE.Audio Bus Writer (2)"(In 1=@n1.Out, In 0=@n2.Out) {also {defaults {Audio Bus=&AB_Wind}}}
   InitVariable.Float()
   TriggerDelay_0 = UE.Trigger Delay(
    In=^UE.Source.OnPlay
    Delay Time=UE.Subtract:Time(AdditionalOperands=^Release Time, PrimaryOperand=UE.ConversionFloatToTime(In=^Duration).Out) {also {version [1,1]}}.Out
   )
   ADSREnvelopeADSREnvelope_0 = ADSR Envelope.ADSR Envelope:Audio(
    Trigger Attack=^UE.Source.OnPlay
    Release Time=UE.Add:Time(PrimaryOperand=^Release Time) {also {defaults {AdditionalOperands=2}}}.Out
    Trigger Release=@TriggerDelay_0.Out
   ) {also {defaults {Decay Curve=0.5, Sustain Level=1}}}
   n3 = &ModuloFloat(A=^UE.Listener.Orientation.Azimuth) {also {defaults {B=360}}}
   n4 = UE.Add:Float(AdditionalOperands=^StereoSpread, PrimaryOperand=@n3.Out)
   n5 = &CompareFloat(A=@n4.Out) {also {defaults {B=180, Type=3}}}
   n6 = UE.Add:Float(
    PrimaryOperand=@n4.Out
    AdditionalOperands=UE.Multiply:Float(
     PrimaryOperand=Convert.Float:Bool(Bool=@n5.Result).Float
     AdditionalOperands=UE.Multiply:Float() {also {defaults {AdditionalOperands=-2, PrimaryOperand=180}}}.Out
    ).Out
   )
   n7 = UE.Multiply:Float() {also {defaults {AdditionalOperands=-1, PrimaryOperand=180}}}
   n8 = UE.Multiply:Float(
    PrimaryOperand=Convert.Float:Bool(Bool=&CompareFloat(A=@n4.Out, B=@n7.Out) {also {defaults {Type=2}}}.Result).Float
    AdditionalOperands=UE.Multiply:Float() {also {defaults {AdditionalOperands=2, PrimaryOperand=180}}}.Out
   )
   CrossfadeTriggerRouteFloat5_1 = "Crossfade.Trigger Route (Float, 5):Float"(Crossfade Value=MapRange.MapRange:Float(In=UE.Add:Float(PrimaryOperand=@n6.Out, AdditionalOperands=@n8.Out).Out) {also {defaults {In Range A=-180, In Range B=180, Out Range B=4}}}.Out Value) {also {defaults {In 0=0, In 1=-1, In 2=0, In 3=1, In 4=0}}}
   StereoPanner_0 = UE.Stereo Panner(Pan Amount=@CrossfadeTriggerRouteFloat5_1.Out, In=@n1.Out)
   n9 = UE.Subtract:Float(AdditionalOperands=^StereoSpread, PrimaryOperand=@n3.Out)
   n10 = &CompareFloat(A=@n9.Out) {also {defaults {B=180, Type=3}}}
   n11 = UE.Add:Float(
    AdditionalOperands=UE.Multiply:Float(
     PrimaryOperand=Convert.Float:Bool(Bool=@n10.Result).Float
     AdditionalOperands=UE.Multiply:Float() {also {defaults {AdditionalOperands=-2, PrimaryOperand=180}}}.Out
    ).Out
    PrimaryOperand=@n9.Out
   )
   n12 = UE.Multiply:Float() {also {defaults {AdditionalOperands=-1, PrimaryOperand=180}}}
   n13 = UE.Multiply:Float(
    PrimaryOperand=Convert.Float:Bool(Bool=&CompareFloat(B=@n12.Out, A=@n9.Out) {also {defaults {Type=2}}}.Result).Float
    AdditionalOperands=UE.Multiply:Float() {also {defaults {AdditionalOperands=2, PrimaryOperand=180}}}.Out
   )
   CrossfadeTriggerRouteFloat5_0 = "Crossfade.Trigger Route (Float, 5):Float"(Crossfade Value=MapRange.MapRange:Float(In=UE.Add:Float(PrimaryOperand=@n11.Out, AdditionalOperands=@n13.Out).Out) {also {defaults {In Range A=-180, In Range B=180, Out Range B=4}}}.Out Value) {also {defaults {In 0=0, In 1=-1, In 2=0, In 3=1, In 4=0}}}
   StereoPanner_1 = UE.Stereo Panner(Pan Amount=@CrossfadeTriggerRouteFloat5_0.Out, In=@n2.Out)
   AudioMixerAudioMixerStereo2_0 = "AudioMixer.Audio Mixer (Stereo, 2)"(In 1 L=@StereoPanner_0.Out Left, In 1 R=@StereoPanner_0.Out Right, In 0 L=@StereoPanner_1.Out Left, In 0 R=@StereoPanner_1.Out Right)
   Multiply_0 = UE.Multiply:Audio(PrimaryOperand=@ADSREnvelopeADSREnvelope_0.Out Envelope, AdditionalOperands=@AudioMixerAudioMixerStereo2_0.Out L)
   Multiply_1 = UE.Multiply:Audio(PrimaryOperand=@ADSREnvelopeADSREnvelope_0.Out Envelope, AdditionalOperands=@AudioMixerAudioMixerStereo2_0.Out R)
  ]
  links [
   {dst="@UE.Source.OneShot.OnFinished", src=ADSREnvelopeADSREnvelope_0.On Done}
   {dst="@UE.OutputFormat.Stereo.Audio:0", src=Multiply_0.Out}
   {dst="@UE.OutputFormat.Stereo.Audio:1", src=Multiply_1.Out}
  ]
 }
]
properties {
 assetImportData=&self:AssetImportData
 bOverrideConcurrency=true
 numChannels=2
 outputFormat=Stereo
 sampleRate=48000
 soundClassObject=&Fort_Ambient_Outdoors_Athena
 concurrencyOverrides {maxCount=1, resolutionRule=StopOldest}
}
